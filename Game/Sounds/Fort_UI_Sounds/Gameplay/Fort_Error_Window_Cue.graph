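# Sound cue for an error window in the gameplay UI, assigned to the Fort_UI sound class.
# A mixer layers a Card_Summary store wave with a UI neutral update wave that starts after a short fixed delay, each at reduced volume.
# A concurrency override allows one instance at a time and stops the oldest when a new one starts.
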SoundCue /Game/Sounds/Fort_UI_Sounds/Gameplay/Fort_Error_Window_Cue
root=Mixer_0
nodes [
 {id=Delay_0, children [WavePlayer_0], properties {delayMax=0.1, delayMin=0.1}}
 {id=Mixer_0, children [WavePlayer_13, Delay_0], properties {inputVolume [0.7,0.6]}}
 {id=WavePlayer_0, properties {soundWaveAssetPtr=&/fort_UI_Neutral_Update}}
 {id=WavePlayer_13, properties {soundWaveAssetPtr=/Game/Sounds/Fort_UI_Sounds/FrontEnd/Store/wav/Card_Summary}}
]
properties {
 bOverrideConcurrency=true
 maxDistance=2097152
 soundClassObject=&Fort_UI
 volumeMultiplier=0.55
 attenuationOverrides {falloffDistance=1600}
 concurrencyOverrides {maxCount=1, resolutionRule=StopOldest}
}
