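# Translucent UI material for the error window that draws a hue-shifted llama icon with a procedural eye whose pupil follows a LookLocation parameter.
# Sphere masks on remapped UVs form a fixed eye white and a smaller pupil, offset by LookLocation and clipped to the eye.
# The eye white is added to the eyeless llama texture, tinted through ColorShift, then the pupil is cut out as black.
# Rotation and debug sine-time branches do not feed this result.
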Material /Game/UI/Foundation/Modals/ErrorWindow/Materials/M_UI_TrackingLlama
graph {
 properties {
  bUsedWithStaticMesh=false
  blendMode=BLEND_Translucent
  materialDomain=MD_UI
  pixelDepthOffsetMode=PDOM_Legacy
  usedShadingModels=Default Lit
 }
 nodes [
  Constant2Vector()
  Constant3Vector()
  n1 = VectorParameter() {parameterName=LookLocation, defaultValue=<rgba 0 0 0 1>}
  LinearInterpolate(Alpha=@n1.R) {constA=-0.1, constB=0.1}
  LinearInterpolate(Alpha=MaterialFunctionCall(Speed=Constant() {r=0.2}) {materialFunction=&DebugTimeSine}.Result)
  MaterialFunctionCall() {materialFunction=&DebugTimeSine}
  n2 = Constant()
  n3 = MaterialFunctionCall(Float2=RotateAboutAxis(
    NormalizedRotationAxis=Constant3Vector() {constant=<rgba 0 0 1 0>}
    RotationAngle=ScalarParameter() {parameterName=RotAngle}
    PivotPoint=Constant3Vector()
    Position=MaterialFunctionCall(X=@n1.R, Y=@n1.R, Z=@n2) {materialFunction=&MakeFloat3}.Result
   )) {materialFunction=&BreakOutFloat2Components}
  MaterialFunctionCall(X=@n3.R, Y=@n3.R) {materialFunction=&MakeFloat2}
  n4 = Subtract(A=Multiply(A=TextureCoordinate()) {constB=2})
  n5 = Constant2Vector() {g=0.0381, r=-0.1714}
  n6 = SphereMask(A=@n4, B=@n5, Radius=Constant() {r=0.15}, Hardness=Constant() {r=0.9})
  n7 = MaterialFunctionCall(X=@n1.R, Y=@n1.R) {materialFunction=&MakeFloat2}
  n8 = Multiply(A=@n6, B=SphereMask(A=@n4, B=Add(A=@n5, B=@n7.Result), Radius=Constant() {r=0.05}, Hardness=Constant() {r=0.9}))
  n9 = ScalarParameter() {parameterName=ColorShift}
  n10 = TextureSample() {texture=&/Icon-Llama-NoEyes}
  Multiply(
   A=OneMinus(None=Clamp(None=@n8))
   B=Add(
    A=Subtract(A=@n6, B=@n8)
    B=MaterialFunctionCall(Hue Shift Percentage=@n9, Texture=@n10.RGB) {materialFunction=&HueShift}.Result
   )
  )
 ]
}
