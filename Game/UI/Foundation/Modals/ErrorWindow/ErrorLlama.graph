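# ErrorLlama is an error-window widget showing a small llama image with randomized color, whose look direction follows the mouse or gamepad left stick.
# Construct creates a dynamic instance of M_UI_TrackingLlama, picks a stepped random ColorShift, and subscribes to CommonInput method changes.
# Tick derives a scaled direction from cursor-to-center offset or stick, corrects for texture flip and widget rotation, then interpolates it into the material's LookLocation parameter.
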WidgetBlueprint /Game/UI/Foundation/Modals/ErrorWindow/ErrorLlama
graphs [
 EventGraph {nodes [
   RandomizeColor() {is=title K2Node_CustomEvent}
   n1 = Get MID_Llama() -> (MID_Llama:Material Instance Dynamic Object Reference)
   > Is Valid(InputObject=@n1)
   > SetScalarParameterValue(
    self=@n1
    ParameterName=ColorShift
    Value=Select!(
     Option 0:double=0
     Option 1:double=0.1
     Option 2:double=0.2
     Option 3:double=0.3
     Option 4:double=0.4
     Option 5:double=0.5
     Option 6:double=0.6
     Option 7:double=0.7
     Option 8:double=0.8
     Option 9:double=0.9
     Index:int=RandomIntegerInRange!(Min:int=0, Max:int=9) -> (ReturnValue:int=0)
    ) -> (ReturnValue:double)
   )
   n2 = HandleNormalizedInputDirection!() -> (OutputDelegate, then, 2DInputVec:Vector 2D Structure) {is=title K2Node_CustomEvent}
   n3 = Get RenderTransform() -> (RenderTransform:Widget Transform Structure)
   n4 = BreakVector2D(InVec=@n2.2DInputVec)
   n5 = Select!(Option 0=1, Option 1=-1, Index:bool=Get bFlippedTexture() -> (bFlippedTexture:bool)) -> (ReturnValue)
   n6 = BreakVector(InVec=Rotate Vector Around Axis!(
     InVect_X:double=float * float(A=@n4, B=@n5)
     InVect_Y:double=@n4.Y
     InVect_Z:double=0
     AngleDeg:float=Break Widget Transform!(WidgetTransform:"Widget Transform Structure (by ref)"=@n3) -> (Translation:Vector 2D Structure, Scale:Vector 2D Structure, Shear:Vector 2D Structure, Angle:float=~) {is=title K2Node_BreakStruct}.Angle
     Axis:Vector="0, 0,1"
    ) -> (ReturnValue:Vector))
   > n7 = Set CurLookAt2D(CurLookAt2D:Vector 2D Structure=Vector2DInterpTo!(
     Current:Vector 2D Structure=Get CurLookAt2D() -> (CurLookAt2D:Vector 2D Structure)
     Target:Vector 2D Structure=MakeVector2D(X=@n6, Y=@n6.Y)
     DeltaTime:float=GetWorldDeltaSeconds!() -> (ReturnValue:double)
     InterpSpeed:float=Get EyeInterpSpeed() -> (EyeInterpSpeed:double)
    ) -> (ReturnValue:Vector 2D Structure)) -> (Output_Get:Vector 2D Structure)
   n8 = BreakVector2D(InVec=@n7)
   > SetVectorParameterValue(
    self=Get MID_Llama()
    ParameterName=LookLocation
    Value=Make Linear Color!(R:float=@n8, G:float=@n8.Y, B:float=0, A:float=0) -> (LinearColor:Linear Color Structure) {is=title K2Node_MakeStruct}
   )
   Event Construct()
   > n9 = CreateDynamicMaterialInstance!(execute, Parent=/Game/UI/Foundation/Modals/ErrorWindow/Materials/M_UI_TrackingLlama, OptionalName=None, CreationFlags=None) -> (then, ReturnValue)
   > n10 = Set MID_Llama(MID_Llama:Material Instance Dynamic Object Reference=@n9) -> (Output_Get:Material Instance Dynamic Object Reference)
   > SetBrushFromMaterial(self=Get Image_0() -> (Image_0:Image Object Reference), Material=@n10)
   > RandomizeColor() {is=title K2Node_CallFunction}
   n11 = Get CommonInputSubsystem!() -> (ReturnValue:Common Input Object Reference) {is=title K2Node_GetBlueprintContextEx}
   > Set CurInputType(CurInputType:ECommonInputType Enum=GetCurrentInputType(self=@n11)) -> (Output_Get:ECommonInputType Enum)
   > Bind Event to On Input Method Changed(self=@n11, Delegate=Create Event())
   n12 = Event Tick()
   > n13 = Switch on ECommonInputType!(execute, Selection:ECommonInputType Enum=Get CurInputType() -> (CurInputType:ECommonInputType Enum)) -> (MouseAndKeyboard:Exec, Gamepad:Exec, Touch:Exec, Count:Exec) {is=title K2Node_SwitchEnum}
   n14 = GetInputAnalogStickState!(
    self:Player Controller Object Reference=GetPlayerController!(PlayerIndex:int=0) -> (ReturnValue:Player Controller Object Reference)
    WhichStick:EControllerAnalogStick Enum=CAS_LeftStick
   ) -> (StickX:float, StickY:float)
   n15 = float * float(A=@n14.StickY, B=-1)
   HandleNormalizedInputDirection!(
    execute=@n13.Gamepad
    self
    2DInputVec:Vector 2D Structure=vector2d * float!(A:Vector 2D Structure=MakeVector2D(X=@n14, Y=@n15), B:double=Get MaterialInputScale() -> (MaterialInputScale:double)) -> (ReturnValue:Vector 2D Structure)
   ) -> (then) {is=title K2Node_CallFunction}
   n16 = GetMousePositionOnPlatform!(execute=@n13.MouseAndKeyboard) -> (then, ReturnValue:Vector 2D Structure)
   n17 = GetLocalSize!(Geometry:"Geometry Structure (by ref)"=@n12.MyGeometry) -> (ReturnValue:Vector 2D Structure)
   n18 = vector2d - vector2d!(
    A:Vector 2D Structure=@n16
    B:Vector 2D Structure="Local to Absolute (Coordinates)"!(
     Geometry:"Geometry Structure (by ref)"=@n12.MyGeometry
     LocalCoordinate:Vector 2D Structure=vector2d / float!(A:Vector 2D Structure=@n17, B:double=2) -> (ReturnValue:Vector 2D Structure)
    ) -> (ReturnValue:Vector 2D Structure)
   ) -> (ReturnValue:Vector 2D Structure)
   > HandleNormalizedInputDirection(2DInputVec=vector2d * float(A=Normalize 2D!(A:Vector 2D Structure=@n18) -> (ReturnValue:Vector 2D Structure), B=Get MaterialInputScale()))
   RandomInteger!(Max:int=0) -> (ReturnValue:int=0)
   Event PreConstruct()
  ]}
 HandleInputTypeChanged {nodes [
   n1 = HandleInputTypeChanged!() -> (then, NewInputType:ECommonInputType Enum) {is=title K2Node_FunctionEntry}
   > Set CurInputType(CurInputType=@n1)
  ]}
]
variables [
 bFlippedTexture {type=boolean}
 CurInputType {type=ECommonInputType, default=MouseAndKeyboard}
 CurLookAt2D {type=Vector2D, default=<0 0>}
 EyeInterpSpeed {type=number, default=8}
 MaterialInputScale {type=number, default=0.1}
 MID_Llama {type=MaterialInstanceDynamic}
]
widgets {
 info {inheritedWidgetCount=0, parentClass=&UserWidget, rootWidgetClass=&SizeBox, widgetCount=2}
 widgets [
  SizeBox_0 {
   widgetClassPath=&SizeBox
   widget=SizeBox_0
   properties {bOverride_HeightOverride=true, bOverride_WidthOverride=true, heightOverride=128, widthOverride=128}
  }
  Image_0 {
   widgetClassPath=&Image
   bIsVariable=true
   parent=SizeBox_0
   widget=Image_0
   properties {brush {
     bIsDynamicallyLoaded=false
     drawAs=Image
     imageType=NoImage
     mirroring=NoMirror
     resourceName=None
     resourceObject=&M_UI_TrackingLlama
     tiling=NoTile
     imageSize=<32 32>
     margin=<ltrb 0 0 0 0>
     outlineSettings {
      bUseBrushTransparency=false
      roundingType=HalfHeightRadius
      width=0
      color {colorUseRule=UseColor_Specified, specifiedColor=<rgba 0 0 0 0>}
      cornerRadii=<0 0 0 0>
     }
     tintColor {colorUseRule=UseColor_Specified, specifiedColor=<rgba 1 1 1 1>}
     uVRegion {bIsValid=false, max=<0 0>, min=<0 0>}
    }}
   slot {class=SizeBoxSlot}
  }
 ]
}
