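# Translucent UI material for the error pop-up border, drawing two horizontal glowing lines.
# OpenScalar spreads the lines from vertical center toward top and bottom, while RevealMask drives a horizontal clip around the center.
# DrawLine-2D strokes taper toward their ends, combined with stretched radial gradients clipped to each line's outer side.
# One sum tints the gradients blue, another applies the reveal mask; several leftover-looking debug and gradient nodes remain.
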Material /Game/UI/Foundation/Modals/ErrorWindow/Materials/M_UI_ErrorPopUpBorder
graph {
 properties {
  bUsedWithStaticMesh=false
  blendMode=BLEND_Translucent
  materialDomain=MD_UI
  pixelDepthOffsetMode=PDOM_Legacy
  usedShadingModels=Default Lit
 }
 nodes [
  n1 = Constant2Vector() {g=1, r=1.5}
  n2 = ScalarParameter() {defaultValue=0.98, parameterName=OpenScalar}
  n3 = LinearInterpolate(Alpha=@n2) {constA=0.5, constB=0.9}
  n4 = MaterialFunctionCall(X=Constant() {r=0.1}, Y=@n3) {materialFunction=&MakeFloat2}
  n5 = MaterialFunctionCall(X=Constant() {r=0.9}, Y=@n3) {materialFunction=&MakeFloat2}
  n6 = MaterialFunctionCall(Float2=Divide(A=Add(A=@n4.Result, B=@n5.Result))) {materialFunction=&BreakOutFloat2Components}
  n7 = MaterialFunctionCall(X=@n6.R, Y=Subtract(A=@n6.G) {constB=0.08}) {materialFunction=&MakeFloat2}
  n8 = TextureCoordinate()
  n9 = ComponentMask(None=@n8) {g=true}
  n10 = Multiply(
   A=MaterialFunctionCall(
    UVs=MaterialFunctionCall(Texture Scale=@n1) {materialFunction=&ScaleUVsByCenter}.UVs
    CenterPosition=@n7.Result
    Radius=Constant() {r=0.2}
    Density=Constant() {r=1}
   ) {materialFunction=&RadialGradientExponential}.RadialGradientExponential
   B=Ceil(None=Subtract(A=@n9, B=@n3) {constB=0.4})
  )
  n11 = Constant3Vector() {constant=<rgba 0 0.27 1 0>}
  n12 = Multiply(A=@n8) {constB=2}
  n13 = Abs(None=ComponentMask(None=Subtract(A=@n12)) {r=true})
  n14 = MaterialFunctionCall!(
   Coordinates
   P0=@n4.Result
   P1=@n5.Result
   Line Width=LinearInterpolate(A=Constant() {r=0.005}, B=Constant(), Alpha=@n13)
   Edge Falloff
   Square Corners?
  ) {materialFunction=/Engine/Functions/Engine_MaterialFunctions01/Coordinates/DrawLine-2D}
  n15 = Constant() {r=0.001}
  n16 = Constant()
  n17 = LinearInterpolate(
   A=@n14.Result
   B=MaterialFunctionCall(P0=@n4.Result, P1=@n5.Result, Line Width=LinearInterpolate(A=@n15, B=@n16, Alpha=@n13), Edge Falloff=Constant() {r=0.6}) {materialFunction=&DrawLine-2D}.Result
   Alpha=@n13
  )
  n18 = LinearInterpolate(Alpha=@n2) {constA=0.5, constB=0.1}
  n19 = Subtract(A=ComponentMask(None=@n8) {g=true}, B=@n18)
  n20 = Constant2Vector() {g=1, r=2.2}
  n21 = MaterialFunctionCall(X=Constant() {r=0.05}, Y=@n18) {materialFunction=&MakeFloat2}
  n22 = MaterialFunctionCall(X=Constant() {r=0.95}, Y=@n18) {materialFunction=&MakeFloat2}
  n23 = MaterialFunctionCall(Float2=Divide(A=Add(A=@n21.Result, B=@n22.Result))) {materialFunction=&BreakOutFloat2Components}
  n24 = Add(A=@n23.G) {constB=0.08}
  n25 = Multiply(
   A=OneMinus(None=Ceil(None=@n19))
   B=MaterialFunctionCall(
    UVs=MaterialFunctionCall(Texture Scale=@n20) {materialFunction=&ScaleUVsByCenter}.UVs
    CenterPosition=MaterialFunctionCall(X=@n23.R, Y=@n24) {materialFunction=&MakeFloat2}.Result
    Radius=Constant() {r=0.19}
    Density=Constant() {r=1}
   ) {materialFunction=&RadialGradientExponential}.RadialGradientExponential
  )
  n26 = Constant()
  n27 = Constant() {r=1.6}
  n28 = MaterialFunctionCall(P0=@n21.Result, P1=@n22.Result, Line Width=LinearInterpolate(A=Constant() {r=0.005}, B=@n26, Alpha=@n13), Edge Falloff=@n27) {materialFunction=&DrawLine-2D}
  n29 = Constant() {r=0.001}
  n30 = LinearInterpolate(
   A=@n28.Result
   B=MaterialFunctionCall(P0=@n21.Result, P1=@n22.Result, Line Width=LinearInterpolate(A=@n29, B=@n26, Alpha=@n13), Edge Falloff=@n27) {materialFunction=&DrawLine-2D}.Result
   Alpha=@n13
  )
  Add(A=Add(A=Multiply(A=@n10, B=@n11), B=@n17), B=Add(A=Multiply(A=@n11, B=@n25), B=@n30))
  Constant2Vector() {g=0.5, r=0.5}
  Constant2Vector()
  Constant() {r=0.1}
  Constant() {r=0.6}
  Constant() {r=0.9}
  LinearInterpolate(Alpha=@n13) {constA=0.6, constB=0}
  n31 = MaterialFunctionCall(P0=Constant2Vector() {g=0.5}, P1=Constant2Vector() {g=0.5, r=1}, Line Width=Constant() {r=0.36}, Edge Falloff=Constant() {r=0.4}) {materialFunction=&DrawLine-2D}
  LinearInterpolate(A=Constant3Vector() {constant=<rgba 1 1 1 0>}, B=@n11, Alpha=OneMinus(None=@n31.Result))
  MaterialFunctionCall!(UVs, CenterPosition, Radius) {materialFunction=/Engine/Functions/Engine_MaterialFunctions01/Gradient/SphereGradient-2D}
  MaterialFunctionCall() {materialFunction=&DebugTimeSine}
  MaterialFunctionCall() {materialFunction=&DebugTimeSine}
  n32 = Constant() {r=0.2}
  Multiply(A=Power(Base=MaterialFunctionCall(UVs=@n8, CenterPosition=@n7.Result, Radius=@n32) {materialFunction=&SphereGradient-2D}.Result 0-1) {constExponent=4})
  Multiply(A=@n28.Result)
  Multiply(A=@n14.Result)
  Multiply(A=@n10)
  n33 = MaterialFunctionCall(Radius=Constant() {r=0.3}) {materialFunction=&SphereGradient-2D}
  Multiply(A=OneMinus(None=Power(Base=@n33.Result 0-1) {constExponent=6}), B=Multiply(A=Constant() {r=0.15}, B=OneMinus(None=@n31.Result)))
  Multiply(A=@n3)
  n34 = Add(A=@n17, B=@n30)
  n35 = ComponentMask(None=@n8) {r=true}
  n36 = LinearInterpolate(Alpha=LinearInterpolate(Alpha=ScalarParameter() {defaultValue=1, parameterName=RevealMask}) {constA=-0.1}) {constA=0.5, constAlpha=-0.1}
  n37 = Add(A=@n35, B=@n36)
  n38 = Subtract(A=@n35, B=@n36)
  n39 = Ceil(None=@n38)
  Multiply(A=Add(A=Add(A=@n10, B=@n34), B=Add(A=@n25, B=@n30)), B=Multiply(A=Floor(None=@n37), B=OneMinus(None=@n39)))
  Multiply(A=@n37, B=OneMinus(None=@n38))
  Power(Base=Add(A=@n13) {constB=0.01})
 ]
}
