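# Translucent UI material for the error window modal, named as a horizontally fading border.
# One branch remaps UVs around the center with absolute value, then inverts and powers the X channel, giving a falloff that peaks mid-width and fades toward both sides.
# A separate DrawLine-2D call defines a soft vertical center line.
# The graph shows neither branch wired to the other or to a material output.
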Material /Game/UI/Foundation/Modals/ErrorWindow/Materials/M_UI_HorizontalFade_Border
graph {
 properties {
  bUsedWithStaticMesh=false
  blendMode=BLEND_Translucent
  materialDomain=MD_UI
  pixelDepthOffsetMode=PDOM_Legacy
  usedShadingModels=Default Lit
 }
 nodes [
  MaterialFunctionCall!(
   Coordinates
   P0=Constant2Vector() {r=0.5}
   P1=Constant2Vector() {g=1, r=0.5}
   Line Width=Constant() {r=0.2}
   Edge Falloff=Constant() {r=1.5}
   Square Corners?
  ) {materialFunction=/Engine/Functions/Engine_MaterialFunctions01/Coordinates/DrawLine-2D}
  n1 = TextureCoordinate()
  n2 = Abs(None=Subtract(A=Multiply(A=@n1) {constB=2}))
  Power(Base=OneMinus(None=ComponentMask(None=@n2) {r=true}))
 ]
}
